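# Physics asset for the belt on the F_MED_Wombat medium female character body, defining collision and joint limits for secondary belt motion.
# It has two box bodies: a flat, offset box on the dyn_belt_01 bone and a small box on the pelvis.
# A single angular constraint links the belt to the pelvis, keeping swing and twist tight except for a wider range on the second swing axis.
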PhysicsAsset /BRCosmetics/Characters/Player/Female/Medium/Bodies/F_MED_Wombat/Meshes/F_MED_Wombat_Physics_Belt
bodies {dyn_belt_01 [{shapeType=Box, extentX=6, extentY=4.35, extentZ=1, center=<3 0 0>}], pelvis [{shapeType=Box, extentX=1, extentY=1, extentZ=1}]}
constraints [{bone1Name=dyn_belt_01, bone2Name=pelvis, swing1LimitDegrees=5, swing2LimitDegrees=25, twistLimitDegrees=5}]
