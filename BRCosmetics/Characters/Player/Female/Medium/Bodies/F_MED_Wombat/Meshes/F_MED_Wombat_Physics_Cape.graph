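# Physics asset for the cape of the F_MED_Wombat player body, with capsule bodies on the cape bones, pelvis, spine, clavicles and arms.
# Two middle chains hang from poncho_root and two side chains from the clavicle_cape anchors, all three anchors joining spine_05 through free angular constraints.
# Cape links have tight twist limits and asymmetric swing limits, stiffest near the top, with tiny placeholder-sized capsules on anchors and first side links.
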PhysicsAsset /BRCosmetics/Characters/Player/Female/Medium/Bodies/F_MED_Wombat/Meshes/F_MED_Wombat_Physics_Cape
bodies {
 clavicle_cape_l [{shapeType=Capsule, length=0.505, radius=0.505}]
 clavicle_cape_r [{shapeType=Capsule, length=0.505, radius=0.505}]
 clavicle_l [{shapeType=Capsule, length=8, radius=8, center=<9.5225 -0.6882 -0.86>, rotation=<rot 82.9999 10.0002 0>}]
 clavicle_r [{shapeType=Capsule, length=8, radius=8, center=<-9 2 0>, rotation=<rot 83 10 0>}]
 dyn_capeMid_01_l [{shapeType=Capsule, length=15, radius=5, center=<9 0 5>, rotation=<rot -90 0 0>}]
 dyn_capeMid_01_r [{shapeType=Capsule, length=15.4733, radius=4.5267, center=<9 0 5.7053>, rotation=<rot -90 180 180>}]
 dyn_capeMid_02_l [{shapeType=Capsule, length=15, radius=10, center=<9 0 8>, rotation=<rot -90 0 0>}]
 dyn_capeMid_02_r [{shapeType=Capsule, length=15, radius=10, center=<9 0 8>, rotation=<rot -90 0 0>}]
 dyn_capeMid_03_l [{shapeType=Capsule, length=15, radius=10, center=<9 0 8>, rotation=<rot -90 0 0>}]
 dyn_capeMid_03_r [{shapeType=Capsule, length=15, radius=10, center=<9 0 8>, rotation=<rot -90 0 0>}]
 dyn_capeMid_04_l [{shapeType=Capsule, length=15, radius=10, center=<9 0 8>, rotation=<rot -90 0 0>}]
 dyn_capeMid_04_r [{shapeType=Capsule, length=15, radius=10, center=<9 0 8>, rotation=<rot -90 0 0>}]
 dyn_cape_01_l [{shapeType=Capsule, length=0.01, radius=0.01}]
 dyn_cape_01_r [{shapeType=Capsule, length=0.01, radius=0.01}]
 dyn_cape_02_l [{shapeType=Capsule, length=15, radius=5, center=<9 0 6.2746>, rotation=<rot -90 180 180>}]
 dyn_cape_02_r [{shapeType=Capsule, length=15, radius=5, center=<9 0 6.2746>, rotation=<rot -90 180 180>}]
 dyn_cape_03_l [{shapeType=Capsule, length=15, radius=10, center=<9 0 8>, rotation=<rot -90 0 0>}]
 dyn_cape_03_r [{shapeType=Capsule, length=15, radius=10, center=<9 0 8>, rotation=<rot -90 0 0>}]
 dyn_cape_04_l [{shapeType=Capsule, length=15, radius=10, center=<9 0 8>, rotation=<rot -90 0 0>}]
 dyn_cape_04_r [{shapeType=Capsule, length=15, radius=10, center=<9 0 8>, rotation=<rot -90 0 0>}]
 dyn_cape_05_l [{shapeType=Capsule, length=15, radius=10, center=<9 0 8>, rotation=<rot -90 0 0>}]
 dyn_cape_05_r [{shapeType=Capsule, length=15, radius=10, center=<9 0 8>, rotation=<rot -90 0 0>}]
 lowerarm_l [{shapeType=Capsule, length=38, radius=9, center=<17.9477 1.1 -1.3>, rotation=<rot 82 0 3.2>}]
 lowerarm_r [{shapeType=Capsule, length=30, radius=8, center=<-18 0.9 2.16>, rotation=<rot 82.7099 0 3.2>}]
 pelvis [{shapeType=Capsule, length=40, radius=16, center=<-5 2 0>}]
 poncho_root [{shapeType=Capsule, length=0.505, radius=0.505, rotation=<rot 0 -90 0>}]
 spine_01 [{shapeType=Capsule, length=14, radius=13, center=<0 2 0>}]
 spine_02 [{shapeType=Capsule, length=15, radius=12, center=<0 2 0>}]
 spine_03 [{shapeType=Capsule, length=16, radius=12, center=<0 2 0>}]
 spine_04 [{shapeType=Capsule, length=48, radius=12, center=<0 2 0>}]
 spine_05 [
  {shapeType=Capsule, length=13, radius=12, center=<-1.8 2 0>}
  {shapeType=Capsule, length=16.4252, radius=4.3426, shapeName=spine_05_capsule, center=<-0.9801 -8.8665 -0.0707>, rotation=<rot -30 -0.0001 0>}
 ]
 upperarm_l [{shapeType=Capsule, length=23.0902, radius=7.9098, center=<13 0.4123 0>, rotation=<rot 90 0 0>}]
 upperarm_r [{shapeType=Capsule, length=18, radius=8, center=<-13.17 0.8 0>, rotation=<rot 90 0 0>}]
}
constraints [
 {bone1Name=spine_02, bone2Name=spine_01, swing1LimitDegrees=45, swing2LimitDegrees=45, twistLimitDegrees=45}
 {bone1Name=spine_04, bone2Name=spine_03, swing1LimitDegrees=45, swing2LimitDegrees=45, twistLimitDegrees=45}
 {bone1Name=spine_05, bone2Name=spine_04, swing1LimitDegrees=45, swing2LimitDegrees=45, twistLimitDegrees=45}
 {bone1Name=upperarm_l, bone2Name=clavicle_l, swing1LimitDegrees=45, swing2LimitDegrees=45, twistLimitDegrees=45}
 {bone1Name=upperarm_r, bone2Name=clavicle_r, swing1LimitDegrees=45, swing2LimitDegrees=45, twistLimitDegrees=45}
 {
  bone1Name=spine_01
  bone2Name=pelvis
  swing1LimitDegrees=45
  swing1Motion=Free
  swing2LimitDegrees=45
  swing2Motion=Free
  twistLimitDegrees=45
  twistMotion=Free
 }
 {
  bone1Name=lowerarm_l
  bone2Name=upperarm_l
  swing1LimitDegrees=45
  swing1Motion=Free
  swing2LimitDegrees=45
  swing2Motion=Free
  twistLimitDegrees=45
  twistMotion=Free
 }
 {
  bone1Name=lowerarm_r
  bone2Name=upperarm_r
  swing1LimitDegrees=45
  swing1Motion=Free
  swing2LimitDegrees=45
  swing2Motion=Free
  twistLimitDegrees=45
  twistMotion=Free
 }
 {
  bone1Name=clavicle_r
  bone2Name=spine_05
  swing1LimitDegrees=45
  swing1Motion=Free
  swing2LimitDegrees=45
  swing2Motion=Free
  twistLimitDegrees=45
  twistMotion=Free
 }
 {
  bone1Name=clavicle_l
  bone2Name=spine_05
  swing1LimitDegrees=45
  swing1Motion=Free
  swing2LimitDegrees=45
  swing2Motion=Free
  twistLimitDegrees=45
  twistMotion=Free
 }
 {bone1Name=dyn_capeMid_01_l, bone2Name=poncho_root, swing1LimitDegrees=5, swing2LimitDegrees=45, twistLimitDegrees=5}
 {bone1Name=dyn_capeMid_02_l, bone2Name=dyn_capeMid_01_l, swing1LimitDegrees=25, swing2LimitDegrees=45, twistLimitDegrees=5}
 {bone1Name=dyn_capeMid_03_l, bone2Name=dyn_capeMid_02_l, swing1LimitDegrees=25, swing2LimitDegrees=45, twistLimitDegrees=5}
 {bone1Name=dyn_capeMid_04_l, bone2Name=dyn_capeMid_03_l, swing1LimitDegrees=25, swing2LimitDegrees=45, twistLimitDegrees=5}
 {bone1Name=dyn_capeMid_01_r, bone2Name=poncho_root, swing1LimitDegrees=5, swing2LimitDegrees=50, twistLimitDegrees=5}
 {bone1Name=dyn_capeMid_02_r, bone2Name=dyn_capeMid_01_r, swing1LimitDegrees=25, swing2LimitDegrees=45, twistLimitDegrees=5}
 {bone1Name=dyn_capeMid_03_r, bone2Name=dyn_capeMid_02_r, swing1LimitDegrees=25, swing2LimitDegrees=45, twistLimitDegrees=5}
 {bone1Name=dyn_capeMid_04_r, bone2Name=dyn_capeMid_03_r, swing1LimitDegrees=25, swing2LimitDegrees=45, twistLimitDegrees=5}
 {
  bone1Name=clavicle_cape_l
  bone2Name=spine_05
  swing1LimitDegrees=45
  swing1Motion=Free
  swing2LimitDegrees=45
  swing2Motion=Free
  twistLimitDegrees=45
  twistMotion=Free
 }
 {bone1Name=dyn_cape_01_l, bone2Name=clavicle_cape_l, swing1LimitDegrees=45, swing2LimitDegrees=45, twistLimitDegrees=5}
 {bone1Name=dyn_cape_02_l, bone2Name=dyn_cape_01_l, swing1LimitDegrees=10, swing2LimitDegrees=45, twistLimitDegrees=5}
 {bone1Name=dyn_cape_03_l, bone2Name=dyn_cape_02_l, swing1LimitDegrees=25, swing2LimitDegrees=45, twistLimitDegrees=5}
 {bone1Name=dyn_cape_04_l, bone2Name=dyn_cape_03_l, swing1LimitDegrees=25, swing2LimitDegrees=45, twistLimitDegrees=5}
 {bone1Name=dyn_cape_05_l, bone2Name=dyn_cape_04_l, swing1LimitDegrees=25, swing2LimitDegrees=45, twistLimitDegrees=5}
 {
  bone1Name=clavicle_cape_r
  bone2Name=spine_05
  swing1LimitDegrees=45
  swing1Motion=Free
  swing2LimitDegrees=45
  swing2Motion=Free
  twistLimitDegrees=45
  twistMotion=Free
 }
 {bone1Name=dyn_cape_01_r, bone2Name=clavicle_cape_r, swing1LimitDegrees=45, swing2LimitDegrees=45, twistLimitDegrees=5}
 {bone1Name=dyn_cape_02_r, bone2Name=dyn_cape_01_r, swing1LimitDegrees=10, swing2LimitDegrees=45, twistLimitDegrees=5}
 {bone1Name=dyn_cape_03_r, bone2Name=dyn_cape_02_r, swing1LimitDegrees=25, swing2LimitDegrees=45, twistLimitDegrees=5}
 {bone1Name=dyn_cape_04_r, bone2Name=dyn_cape_03_r, swing1LimitDegrees=25, swing2LimitDegrees=45, twistLimitDegrees=5}
 {bone1Name=dyn_cape_05_r, bone2Name=dyn_cape_04_r, swing1LimitDegrees=25, swing2LimitDegrees=45, twistLimitDegrees=5}
 {
  bone1Name=spine_03
  bone2Name=spine_02
  swing1LimitDegrees=45
  swing1Motion=Free
  swing2LimitDegrees=45
  swing2Motion=Free
  twistLimitDegrees=45
  twistMotion=Free
 }
 {
  bone1Name=poncho_root
  bone2Name=spine_05
  swing1LimitDegrees=45
  swing1Motion=Free
  swing2LimitDegrees=45
  swing2Motion=Free
  twistLimitDegrees=45
  twistMotion=Free
 }
]
