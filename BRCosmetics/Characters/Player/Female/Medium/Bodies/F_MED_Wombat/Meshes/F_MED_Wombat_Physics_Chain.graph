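# Physics asset for the Wombat female medium body cosmetic that sets up physics for the chain hanging on the right hip.
# Box and capsule bodies on front, mid and back chain bones each constrain to a tiny MSTR anchor body with limit angles set per segment.
# The anchors link to the pelvis, and front and back segments link to mid, with free angular motion.
# Pelvis and right thigh capsules give the body volumes around the chain.
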PhysicsAsset /BRCosmetics/Characters/Player/Female/Medium/Bodies/F_MED_Wombat/Meshes/F_MED_Wombat_Physics_Chain
bodies {
 dyn_chain_back_01_r [{shapeType=Box, extentX=14, extentY=12.5, extentZ=1.5, center=<-7.4 0 0>, rotation=<rot 0 0 -180>}]
 dyn_chain_back_01_r_MSTR [{shapeType=Box, extentX=0.01, extentY=0.01, extentZ=0.01}]
 dyn_chain_front_01_r [{shapeType=Box, extentX=8.5, extentY=9, extentZ=3.5, center=<-12.7 0 0>, rotation=<rot 0 0 180>}]
 dyn_chain_front_01_r_MSTR [{shapeType=Box, extentX=0.01, extentY=0.01, extentZ=0.01}]
 dyn_chain_mid_01_r [
  {
   shapeType=Capsule
   length=14.176
   radius=1.9858
   shapeName=dyn_chain_mid_01_r_capsule
   center=<-24.578 2.7932 -0.5501>
   rotation=<rot 0 0.0007 79.9999>
  }
  {
   shapeType=Capsule
   length=13.1456
   radius=1.9858
   shapeName=dyn_chain_mid_01_r_capsule
   center=<-24.5778 -9.3446 1.4034>
   rotation=<rot -0.0001 0.001 119.9999>
  }
 ]
 dyn_chain_mid_01_r_MSTR [{shapeType=Box, extentX=0.01, extentY=0.01, extentZ=0.01}]
 pelvis [{shapeType=Capsule, length=11.63, radius=10.18, center=<-6.94 -2.75 -0.029>, rotation=<rot -0.2336 -90 0.149>}]
 thigh_r [{shapeType=Capsule, length=29.8246, radius=7.9254, center=<15.5777 1.3802 0.6182>, rotation=<rot -90 0 179.9999>}]
 thigh_twist_01_r [{shapeType=Capsule, length=13.7538, radius=9.0862, center=<-6.3388 0.8801 0.0116>, rotation=<rot -90 0 -179.9999>}]
}
constraints [
 {
  bone1Name=thigh_twist_01_r
  bone2Name=thigh_r
  swing1LimitDegrees=45
  swing1Motion=Free
  swing2LimitDegrees=45
  swing2Motion=Free
  twistLimitDegrees=45
  twistMotion=Free
 }
 {bone1Name=dyn_chain_front_01_r, bone2Name=dyn_chain_front_01_r_MSTR, swing1LimitDegrees=10, swing2LimitDegrees=75, twistLimitDegrees=5}
 {bone1Name=dyn_chain_mid_01_r, bone2Name=dyn_chain_mid_01_r_MSTR, swing1LimitDegrees=10, swing2LimitDegrees=75, twistLimitDegrees=5}
 {bone1Name=dyn_chain_back_01_r, bone2Name=dyn_chain_back_01_r_MSTR, swing1LimitDegrees=45, swing2LimitDegrees=45, twistLimitDegrees=15}
 {
  bone1Name=dyn_chain_back_01_r
  bone2Name=dyn_chain_mid_01_r
  swing1LimitDegrees=45
  swing1Motion=Free
  swing2LimitDegrees=45
  swing2Motion=Free
  twistLimitDegrees=45
  twistMotion=Free
 }
 {
  bone1Name=dyn_chain_front_01_r
  bone2Name=dyn_chain_mid_01_r
  swing1LimitDegrees=45
  swing1Motion=Free
  swing2LimitDegrees=45
  swing2Motion=Free
  twistLimitDegrees=45
  twistMotion=Free
 }
 {
  bone1Name=thigh_r
  bone2Name=pelvis
  swing1LimitDegrees=45
  swing1Motion=Free
  swing2LimitDegrees=45
  swing2Motion=Free
  twistLimitDegrees=45
  twistMotion=Free
 }
 {
  bone1Name=dyn_chain_front_01_r_MSTR
  bone2Name=pelvis
  swing1LimitDegrees=45
  swing1Motion=Free
  swing2LimitDegrees=45
  swing2Motion=Free
  twistLimitDegrees=45
  twistMotion=Free
 }
 {
  bone1Name=dyn_chain_mid_01_r_MSTR
  bone2Name=pelvis
  swing1LimitDegrees=45
  swing1Motion=Free
  swing2LimitDegrees=45
  swing2Motion=Free
  twistLimitDegrees=45
  twistMotion=Free
 }
 {
  bone1Name=dyn_chain_back_01_r_MSTR
  bone2Name=pelvis
  swing1LimitDegrees=45
  swing1Motion=Free
  swing2LimitDegrees=45
  swing2Motion=Free
  twistLimitDegrees=45
  twistMotion=Free
 }
]
