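# Post-copy Anim Blueprint for the F_MED_Wombat body that simulates its cape, chain and belt.
# The copied, constrained pose runs through RigidBody sims, with backbling selecting the cape physics asset and a Stage2 variant tag gating cape and chain sims.
# Swimming and skydiving states drive cape gravity, sometimes noise-modulated, while skydiving force feeds chain wind.
# Right-thigh swing and crouching drive corrective chain rotations, which weighted constraints spread across chain bones.
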AnimBlueprint /BRCosmetics/Characters/Player/Female/Medium/Bodies/F_MED_Wombat/Meshes/F_MED_Wombat_AnimBP
target_skeleton=&/F_MED_Wombat_Skeleton
graphs [
 AnimGraph {nodes [
   n1 = Get GravityOverride()
   n2 = RigidBody!(ComponentPose=Local To Component(LocalPose=Use cached pose 'preSimCache'()), Alpha, OverrideWorldGravity:Vector=@n1, ExternalForce) -> (Pose) {
    bClampLinearTranslationLimitToRefPose=true
    bOverrideWorldGravity=true
    overridePhysicsAsset=&/F_MED_Wombat_Physics_Cape
    simulationSpace=BaseBoneSpace
    alphaBoolBlend {blendOption=Linear}
    baseBoneRef {boneName=pelvis}
    componentAppliedLinearAccClamp=<3000 3000 3000>
    componentLinearAccScale=<0.4 0.4 0.4>
    componentLinearVelScale=<0.4 0.4 0.4>
    overrideWorldGravity {z=-980}
    simSpaceSettings {
     maxAngularAcceleration=1000
     maxAngularVelocity=3
     maxLinearAcceleration=700
     maxLinearVelocity=500
     velocityScaleZ=0
     worldAlpha=1
     externalLinearDragV=<1 1>
    }
   }
   n3 = Component To Local(ComponentPose=@n2)
   n4 = RigidBody(ComponentPose=Local To Component(LocalPose=Use cached pose 'preSimCache'()), OverrideWorldGravity=@n1) {
    bClampLinearTranslationLimitToRefPose=true
    bOverrideWorldGravity=true
    overridePhysicsAsset=&/F_MED_Wombat_Physics_CapeBackpack
    simulationSpace=BaseBoneSpace
    alphaBoolBlend {blendOption=Linear}
    baseBoneRef {boneName=pelvis}
    componentAppliedLinearAccClamp=<3000 3000 3000>
    componentLinearAccScale=<0.3 0.3 0.3>
    componentLinearVelScale=<0.3 0.3 0.3>
    overrideWorldGravity {z=-980}
    simSpaceSettings {
     maxAngularAcceleration=1000
     maxAngularVelocity=3
     maxLinearAcceleration=700
     maxLinearVelocity=500
     velocityScaleZ=0
     worldAlpha=1
     externalLinearDragV=<1 1>
    }
   }
   n5 = Component To Local(ComponentPose=@n4)
   n6 = Get isWearingBackbling()
   n7 = RigidBody(
    ComponentPose=Local To Component(LocalPose=Blend Poses by bool(BlendPose_0=@n5, BlendPose_1=@n3, bActiveValue=@n6) {blendType=Linear})
    OverrideWorldGravity="0,0,-980"
    ExternalForce=Get windVector() -> (windVector:Vector)
   ) {
    bClampLinearTranslationLimitToRefPose=true
    bOverrideWorldGravity=true
    overridePhysicsAsset=&/F_MED_Wombat_Physics_Chain
    simulationSpace=BaseBoneSpace
    alphaBoolBlend {blendOption=Linear}
    baseBoneRef {boneName=spine_03}
    componentAppliedLinearAccClamp=<2900 2900 2900>
    componentLinearAccScale=<0.2 0.2 0.2>
    componentLinearVelScale=<0.7 0.7 0.7>
    overrideWorldGravity {z=-980}
   }
   n8 = Component To Local(ComponentPose=@n7)
   n9 = Use cached pose 'preSimCache'()
   n10 = Get cloak_active() -> (cloak_active:bool)
   n11 = RigidBody(
    ComponentPose=Local To Component(LocalPose=Blend Poses by bool(BlendPose_0=@n8, BlendPose_1=@n9, bActiveValue=@n10) {blendType=Linear})
    OverrideWorldGravity="0,0,-980"
   ) {
    bClampLinearTranslationLimitToRefPose=true
    bOverrideWorldGravity=true
    overridePhysicsAsset=&/F_MED_Wombat_Physics_Belt
    simulationSpace=BaseBoneSpace
    alphaBoolBlend {blendOption=Linear}
    baseBoneRef {boneName=pelvis}
    componentAppliedLinearAccClamp=<3000 3000 3000>
    componentLinearAccScale=<0.4 0.4 0.4>
    componentLinearVelScale=<0.4 0.4 0.4>
    overrideWorldGravity {z=-980}
    simSpaceSettings {
     maxAngularAcceleration=1000
     maxAngularVelocity=3
     maxLinearAcceleration=700
     maxLinearVelocity=500
     velocityScaleZ=0
     worldAlpha=1
     externalLinearDragV=<1 1>
    }
   }
   n12 = "Transform (Modify) Bone - Bone: dyn_chain_back_01_r"(
    ComponentPose=@n11
    Alpha=float * float(A=Get thighSwingDelta_r_post() -> (thighSwingDelta_r_post:double), B=0.5)
    Translation="0,0,0"
    Rotation="30,22,0"
    Scale="1,1,1"
   ) {
    rotationMode=BMM_Additive
    rotationSpace=BCS_BoneSpace
    scaleMode=BMM_Replace
    scaleSpace=BCS_BoneSpace
    translationMode=BMM_Additive
    translationSpace=BCS_BoneSpace
    alphaBoolBlend {blendOption=Linear}
    boneToModify {boneName=dyn_chain_back_01_r}
    rotation {pitch=30, yaw=22}
   }
   n13 = "Transform (Modify) Bone - Bone: dyn_chain_mid_01_r"(ComponentPose=@n12, Alpha=0.5, Translation="0,0,0", Rotation="10,0.000,0", Scale="1,1,1") {
    rotationMode=BMM_Additive
    rotationSpace=BCS_BoneSpace
    scaleMode=BMM_Additive
    scaleSpace=BCS_BoneSpace
    translationMode=BMM_Additive
    translationSpace=BCS_BoneSpace
    alphaBoolBlend {blendOption=Linear}
    boneToModify {boneName=dyn_chain_mid_01_r}
    rotation {pitch=10}
   }
   n14 = "Transform (Modify) Bone - Bone: dyn_chain_front_01_r"(
    ComponentPose=@n13
    Alpha=Get thighSwingDelta_r_post_mid_rotation() -> (thighSwingDelta_r_post_mid_rotation:double)
    Translation="0,0,0"
    Rotation="30,0.000,0"
    Scale="1,1,1"
   ) {
    rotationMode=BMM_Additive
    rotationSpace=BCS_BoneSpace
    scaleMode=BMM_Additive
    scaleSpace=BCS_BoneSpace
    translationMode=BMM_Additive
    translationSpace=BCS_BoneSpace
    alphaBoolBlend {blendOption=Linear}
    boneToModify {boneName=dyn_chain_front_01_r}
    rotation {pitch=30}
   }
   n15 = "Constraint - chain_09_r_DSP (P)"(ComponentPose=@n14, Alpha=1, ConstraintWeights_0=1) {
    alphaBoolBlend {blendOption=Linear}
    boneToModify {boneName=chain_09_r_DSP}
    constraintSetup [{targetBone {boneName=dyn_chain_back_01_r}}]
    constraintWeights [1]
   }
   n16 = "Constraint - chain_08_r_DSP (PPP)"!(ComponentPose=@n15, Alpha=1, ConstraintWeights_0=0.2, ConstraintWeights_1:float=0, ConstraintWeights_2:float=0.8) -> (Pose) {
    alphaBoolBlend {blendOption=Linear}
    boneToModify {boneName=chain_08_r_DSP}
    constraintSetup [{targetBone {boneName=dyn_chain_back_01_r}}, {targetBone {boneName=dyn_chain_mid_01_r}}, {targetBone {boneName=chain_09_r_DSP}}]
    constraintWeights [1,1,1]
   }
   n17 = "Constraint - chain_07_r_DSP (PPP)"!(ComponentPose=@n16, Alpha=1, ConstraintWeights_0=0.2, ConstraintWeights_1:float=0.2, ConstraintWeights_2:float=0.6) -> (Pose) {
    alphaBoolBlend {blendOption=Linear}
    boneToModify {boneName=chain_07_r_DSP}
    constraintSetup [{targetBone {boneName=dyn_chain_back_01_r}}, {targetBone {boneName=dyn_chain_mid_01_r}}, {targetBone {boneName=chain_08_r_DSP}}]
    constraintWeights [1,1,1]
   }
   n18 = "Constraint - chain_04_r_DSP (P)"(ComponentPose=@n17, Alpha=1, ConstraintWeights_0=1) {
    alphaBoolBlend {blendOption=Linear}
    boneToModify {boneName=chain_04_r_DSP}
    constraintSetup [{targetBone {boneName=dyn_chain_mid_01_r}}]
    constraintWeights [1]
   }
   n19 = "Constraint - chain_03_r_DSP (PP)"!(ComponentPose=@n18, Alpha=1, ConstraintWeights_0=0.2, ConstraintWeights_1:float=0.8) -> (Pose) {
    alphaBoolBlend {blendOption=Linear}
    boneToModify {boneName=chain_03_r_DSP}
    constraintSetup [{targetBone {boneName=dyn_chain_front_01_r}}, {targetBone {boneName=dyn_chain_mid_01_r}}]
    constraintWeights [1,1]
   }
   n20 = "Constraint - chain_01_r_DSP (PP)"!(ComponentPose=@n19, Alpha=1, ConstraintWeights_0=0.7, ConstraintWeights_1:float=0.3) -> (Pose) {
    alphaBoolBlend {blendOption=Linear}
    boneToModify {boneName=chain_01_r_DSP}
    constraintSetup [{targetBone {boneName=dyn_chain_front_01_r}}, {targetBone {boneName=dyn_chain_mid_01_r}}]
    constraintWeights [1,1]
   }
   n21 = "Constraint - chain_02_r_DSP (PP)"!(ComponentPose=@n20, Alpha=1, ConstraintWeights_0=0.3, ConstraintWeights_1:float=0.7) -> (Pose) {
    alphaBoolBlend {blendOption=Linear}
    boneToModify {boneName=chain_02_r_DSP}
    constraintSetup [{targetBone {boneName=dyn_chain_front_01_r}}, {targetBone {boneName=dyn_chain_mid_01_r}}]
    constraintWeights [1,1]
   }
   n22 = "Constraint - chain_06_r_DSP (PPPP)"!(
    ComponentPose=@n21
    Alpha=1
    ConstraintWeights_0=0
    ConstraintWeights_1:float=0
    ConstraintWeights_2:float=0.8
    ConstraintWeights_3:float=0.2
   ) -> (Pose) {
    alphaBoolBlend {blendOption=Linear}
    boneToModify {boneName=chain_06_r_DSP}
    constraintSetup [
     {targetBone {boneName=dyn_chain_back_01_r}}
     {targetBone {boneName=dyn_chain_mid_01_r}}
     {targetBone {boneName=chain_07_r_DSP}}
     {targetBone {boneName=chain_04_r_DSP}}
    ]
    constraintWeights [1,1,1,1]
   }
   n23 = "Constraint - chain_05_r_DSP (PPPP)"!(
    ComponentPose=@n22
    Alpha=1
    ConstraintWeights_0=0
    ConstraintWeights_1:float=0
    ConstraintWeights_2:float=0.2
    ConstraintWeights_3:float=0.8
   ) -> (Pose) {
    alphaBoolBlend {blendOption=Linear}
    boneToModify {boneName=chain_05_r_DSP}
    constraintSetup [
     {targetBone {boneName=dyn_chain_back_01_r}}
     {targetBone {boneName=dyn_chain_mid_01_r}}
     {targetBone {boneName=chain_06_r_DSP}}
     {targetBone {boneName=chain_04_r_DSP}}
    ]
    constraintWeights [1,1,1,1]
   }
   Output Pose(Result=Component To Local(ComponentPose=@n23)) {layerGroup=None}
   n24 = "Constraint - dyn_chain_front_01_r_MSTR (R)"(ComponentPose=Local To Component(LocalPose=Copy Pose From Mesh() {bUseAttachedParent=true}), Alpha=1, ConstraintWeights_0=1) {
    alphaBoolBlend {blendOption=Linear}
    boneToModify {boneName=dyn_chain_front_01_r_MSTR}
    constraintSetup [{transformType=Rotation, targetBone {boneName=thigh_twist_01_r}}]
    constraintWeights [1]
   }
   n25 = "Constraint - dyn_chain_mid_01_r_MSTR (R)"(ComponentPose=@n24, Alpha=1, ConstraintWeights_0=1) {
    alphaBoolBlend {blendOption=Linear}
    boneToModify {boneName=dyn_chain_mid_01_r_MSTR}
    constraintSetup [{transformType=Rotation, targetBone {boneName=thigh_twist_01_r}}]
    constraintWeights [1]
   }
   n26 = "Transform (Modify) Bone - Bone: dyn_chain_front_01_r_MSTR"(ComponentPose=@n25, Alpha=0, Translation="2,-1.000,0.000", Rotation="40.000,-35,-0.000", Scale="1,1,1") {
    rotationMode=BMM_Additive
    translationMode=BMM_Additive
    alphaBoolBlend {blendOption=Linear}
    boneToModify {boneName=dyn_chain_front_01_r_MSTR}
    rotation {pitch=40, yaw=-35}
    translation=<2 -1>
   }
   n27 = "Constraint - dyn_cape_01_l (PP)"!(ComponentPose=@n26, Alpha=1, ConstraintWeights_0=0.5, ConstraintWeights_1:float=0.5) -> (Pose) {
    alphaBoolBlend {blendOption=Linear}
    boneToModify {boneName=dyn_cape_01_l}
    constraintSetup [{targetBone {boneName=spine_05}}, {targetBone {boneName=clavicle_l}}]
    constraintWeights [1,1]
   }
   n28 = "Constraint - dyn_cape_01_r (PP)"!(ComponentPose=@n27, Alpha=1, ConstraintWeights_0=0.5, ConstraintWeights_1:float=0.5) -> (Pose) {
    alphaBoolBlend {blendOption=Linear}
    boneToModify {boneName=dyn_cape_01_r}
    constraintSetup [{targetBone {boneName=clavicle_r}}, {targetBone {boneName=spine_05}}]
    constraintWeights [1,1]
   }
   Save cached pose 'preSimCache'(Pose=Component To Local(ComponentPose=@n28)) {cacheName=preSimCache}
  ]}
 EventGraph {nodes [
   Event BlueprintUpdateAnimation()
   > n1 = Sequence!(execute) -> (then_0, then_1, then_2:Exec)
   > n2 = Branch(Condition=Get bIsSwimSprinting() -> (bIsSwimSprinting:bool))
   n3 = GetSocketRotation!(self:Scene Component Object Reference=Get Skel_Mesh(), InSocketName:Name=pelvis) -> (ReturnValue:Rotator)
   n4 = PerlinNoise1D!(Value:float=GetGameTimeInSeconds()) -> (ReturnValue:float)
   n5 = float * float(A="Sin (Degrees)"!(A:double=float * float(A=@n4, B=800)) -> (ReturnValue:double), B=-400)
   n6 = vector + vector(
    A:Vector=vector * float(A=GetForwardVector!(InRot:Rotator=@n3) -> (ReturnValue:Vector), B=-300)
    B:Vector=MakeVector(Y=0, Z=float + float(A=@n5, B=200) {is=title K2Node_CommutativeAssociativeBinaryOperator})
   ) -> (ReturnValue:Vector) {is=title K2Node_CommutativeAssociativeBinaryOperator}
   > Set GravityOverride(GravityOverride=@n6)
   n7 = Branch(execute=@n2.else, Condition=Get bIsSkyDiving())
   > n8 = Branch(Condition=Get bIsParachuteOpened())
   Set GravityOverride(execute=[@n8.then, @n7.else], GravityOverride="0, 0,-980")
   n9 = Branch(execute=@n8.else, Condition="float > float"(A=Get PawnVelocityZ(), B=50) {is=title K2Node_CallFunction})
   > Set GravityOverride(GravityOverride="0, 0,-100")
   Set GravityOverride(execute=@n9.else, GravityOverride=@n6)
   n10 = Get SkydivingExternalForceInput()
   Set windVector(
    execute=@n1.then_1
    windVector:Vector=vector * float(
     A=Break Fort Anim Input Skydiving External Force(FortAnimInput_SkydivingExternalForce=@n10) -> (
      bUseSkydivingVectorForce=~
      SkydivingVectorForce=~
      HeadToPelvisDirection=~
      FloatingMultiplier=~
      FloatingAdditive=~
      DivingMultiplier=~
      DivingAdditive=~
      ParachutingMultiplier=~
      ParachutingAdditive=~
     ).SkydivingVectorForce
     B=3
    )
   ) -> (Output_Get:Vector)
   n11 = GetDeltaTransformFromRefPose(self=GetOwnerSkeletalMeshComponent(), BoneName=thigh_r, BaseName=pelvis)
   n12 = "Absolute (Float)"(A=BreakRotator(InRot=BreakTransform(InTransform=@n11).Rotation).Yaw)
   Set thighSwingDelta_r(execute=@n1.then_2, thighSwingDelta_r:double=MapRangeClamped(Value=@n12, InRangeA=50, InRangeB=110)) -> (Output_Get:double)
   n13 = Get bIsCrouching()
   > Set thighSwingDelta_r_post(thighSwingDelta_r_post:double=MapRangeClamped(Value=@n12, InRangeA=50, InRangeB=110, OutRangeA=SelectFloat(A=1, B=0, bPickA=@n13))) -> (Output_Get:double)
   > Set thighSwingDelta_r_post_mid_rotation(thighSwingDelta_r_post_mid_rotation:double=SelectFloat(A=1, B=0.5, bPickA=Get bIsCrouching())) -> (Output_Get:double)
   Event BlueprintInitializeAnimation()
   > Set Skel_Mesh(Skel_Mesh=GetOwnerSkeletalMeshComponent())
   Event OnCharacterCustomizationLoaded()
   > Set isWearingBackbling(isWearingBackbling=CheckForBackbling())
   > n14 = Branch(Condition=CharacterItemDefinitionHasAnyMetaTag(Tags=Make Array("[0]"=(TagName="Cosmetics.Variant.Property.Stage2"))))
   > Set cloak_active(cloak_active:bool=true) -> (Output_Get:bool)
   Set cloak_active(execute=@n14.else, cloak_active=false)
  ]}
]
variables [
 cloak_active {type=boolean}
 GravityOverride {type=Vector}
 isWearingBackbling {type=boolean}
 Skel_Mesh {type=SkeletalMeshComponent}
 thighSwingDelta_r {type=number}
 thighSwingDelta_r_post {type=number}
 thighSwingDelta_r_post_mid_rotation {type=number}
 windVector {type=Vector}
]
