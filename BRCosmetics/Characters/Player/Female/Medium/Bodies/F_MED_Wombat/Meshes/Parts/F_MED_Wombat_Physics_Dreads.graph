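# Physics asset for the dreadlocks part of the female medium Wombat player body, giving the hair physical bodies and joint limits.
# Four strands of four box or capsule bodies form chains from the head, each joint held to tight swing and twist limits.
# Head, upper spine, clavicle and upper arm bodies supply surrounding shapes.
# These body parts are linked by free-rotation constraints, except the left upper arm, which has none.
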PhysicsAsset /BRCosmetics/Characters/Player/Female/Medium/Bodies/F_MED_Wombat/Meshes/Parts/F_MED_Wombat_Physics_Dreads
bodies {
 clavicle_l [{
   shapeType=Capsule
   length=4.2301
   radius=5.2891
   shapeName=clavicle_l_capsule
   center=<5.5947 0.0001 -0.4205>
   rotation=<rot -90 0.005 -0.005>
  }]
 clavicle_r [{
   shapeType=Capsule
   length=3.9231
   radius=5.596
   shapeName=clavicle_l_capsule
   center=<-5.5947 -0.0001 0.4205>
   rotation=<rot -90 -0.0049 -179.9951>
  }]
 dyn_hairStrand1_01 [{shapeType=Box, extentX=5, extentY=2, extentZ=2, center=<2.5 0.0139 0.0219>, rotation=<rot -0.2578 -1.4566 -0.1639>}]
 dyn_hairStrand1_02 [{shapeType=Box, extentX=5, extentY=2, extentZ=2, center=<2.5 -0.1338 0.0163>}]
 dyn_hairStrand1_03 [{
   shapeType=Capsule
   length=2.9708
   radius=1.334
   shapeName=dyn_hairStrand1_03_capsule
   center=<2.4442 0 0.0001>
   rotation=<rot -90 -0.0198 0.0198>
  }]
 dyn_hairStrand1_04 [{
   shapeType=Capsule
   length=3.7166
   radius=1.0897
   shapeName=dyn_hairStrand1_04_capsule
   center=<2.5157 0.1892 -0.0002>
   rotation=<rot 90 0 -0.0002>
  }]
 dyn_hairStrand2_01 [{shapeType=Box, extentX=5, extentY=2, extentZ=2, center=<2.5 -0.0641 0.0019>, rotation=<rot -0.2578 -1.4566 -0.1639>}]
 dyn_hairStrand2_02 [{shapeType=Box, extentX=5, extentY=2, extentZ=2, center=<2.5 0.0204 -0.0522>}]
 dyn_hairStrand2_03 [{shapeType=Capsule, length=4.2017, radius=1.2337, shapeName=dyn_hairStrand2_03_capsule, center=<2.571 0 0>, rotation=<rot 90 -0.0198 -0.0198>}]
 dyn_hairStrand2_04 [{
   shapeType=Capsule
   length=4.286
   radius=1.1722
   shapeName=dyn_hairStrand2_04_capsule
   center=<2.2735 -0.0001 0>
   rotation=<rot 90 0.0396 0.0397>
  }]
 dyn_hairStrand3_01 [{shapeType=Box, extentX=5, extentY=2, extentZ=2, center=<2.5 -0.0139 0.013>, rotation=<rot -0.2578 -1.4566 -0.1639>}]
 dyn_hairStrand3_02 [{shapeType=Box, extentX=5, extentY=2, extentZ=2, center=<2.5 -0.0262 0.0415>}]
 dyn_hairStrand3_03 [{shapeType=Box, extentX=5, extentY=2, extentZ=2, center=<2.5 -0.0947 0.0546>}]
 dyn_hairStrand3_04 [{shapeType=Box, extentX=5, extentY=2, extentZ=2, center=<2.5 -0.1256 -0.0732>}]
 dyn_hairStrand4_01 [{shapeType=Box, extentX=5, extentY=2, extentZ=2, center=<2.5 0.0144 -0.0604>, rotation=<rot -0.2578 -1.4566 -0.1639>}]
 dyn_hairStrand4_02 [{shapeType=Box, extentX=5, extentY=2, extentZ=2, center=<2.5 0.0309 -0.0201>}]
 dyn_hairStrand4_03 [{shapeType=Box, extentX=5, extentY=2, extentZ=2, center=<2.5 -0.0213 0.0392>}]
 dyn_hairStrand4_04 [{shapeType=Box, extentX=5, extentY=2, extentZ=2, center=<2.5 0.0008 -0.038>}]
 head [
  {shapeType=Capsule, length=5.4529, radius=6.3542, shapeName=head_capsule, center=<1.154 4.2699 0.0005>, rotation=<rot -90 -0.0099 0.0099>}
  {shapeType=Box, extentX=1, extentY=1, extentZ=1}
 ]
 spine_05 [{shapeType=Capsule, length=10.9437, radius=9.5131, center=<2.4493 -0.5024 0>}]
 upperarm_l [{shapeType=Capsule, length=18.0337, radius=4.7986, shapeName=upperarm_l_capsule, center=<9.9361 0.0007 0>, rotation=<rot 90 -0.0197 -0.0196>}]
 upperarm_r [
  {
   shapeType=Capsule
   length=17.9872
   radius=4.845
   shapeName=upperarm_l_capsule
   center=<-9.9361 -0.0007 0>
   rotation=<rot 90 0.0297 -179.9702>
  }
  {shapeType=Capsule, length=3.0724, radius=5.8799, shapeName=upperarm_r_capsule, center=<-2.794 0 0.0806>, rotation=<rot -90 0.0396 -0.0396>}
 ]
}
constraints [
 {bone1Name=dyn_hairStrand1_01, bone2Name=head, swing1LimitDegrees=10, swing2LimitDegrees=10, twistLimitDegrees=2}
 {bone1Name=dyn_hairStrand1_02, bone2Name=dyn_hairStrand1_01, swing1LimitDegrees=10, swing2LimitDegrees=10, twistLimitDegrees=2}
 {bone1Name=dyn_hairStrand1_03, bone2Name=dyn_hairStrand1_02, swing1LimitDegrees=10, swing2LimitDegrees=10, twistLimitDegrees=2}
 {bone1Name=dyn_hairStrand1_04, bone2Name=dyn_hairStrand1_03, swing1LimitDegrees=10, swing2LimitDegrees=10, twistLimitDegrees=2}
 {bone1Name=dyn_hairStrand2_01, bone2Name=head, swing1LimitDegrees=10, swing2LimitDegrees=10, twistLimitDegrees=2}
 {bone1Name=dyn_hairStrand2_02, bone2Name=dyn_hairStrand2_01, swing1LimitDegrees=10, swing2LimitDegrees=10, twistLimitDegrees=2}
 {bone1Name=dyn_hairStrand2_03, bone2Name=dyn_hairStrand2_02, swing1LimitDegrees=10, swing2LimitDegrees=10, twistLimitDegrees=2}
 {bone1Name=dyn_hairStrand2_04, bone2Name=dyn_hairStrand2_03, swing1LimitDegrees=10, swing2LimitDegrees=10, twistLimitDegrees=2}
 {bone1Name=dyn_hairStrand4_01, bone2Name=head, swing1LimitDegrees=10, swing2LimitDegrees=10, twistLimitDegrees=2}
 {bone1Name=dyn_hairStrand4_02, bone2Name=dyn_hairStrand4_01, swing1LimitDegrees=10, swing2LimitDegrees=10, twistLimitDegrees=2}
 {bone1Name=dyn_hairStrand4_03, bone2Name=dyn_hairStrand4_02, swing1LimitDegrees=10, swing2LimitDegrees=10, twistLimitDegrees=2}
 {bone1Name=dyn_hairStrand4_04, bone2Name=dyn_hairStrand4_03, swing1LimitDegrees=10, swing2LimitDegrees=10, twistLimitDegrees=2}
 {bone1Name=dyn_hairStrand3_01, bone2Name=head, swing1LimitDegrees=10, swing2LimitDegrees=10, twistLimitDegrees=2}
 {bone1Name=dyn_hairStrand3_02, bone2Name=dyn_hairStrand3_01, swing1LimitDegrees=10, swing2LimitDegrees=10, twistLimitDegrees=2}
 {bone1Name=dyn_hairStrand3_03, bone2Name=dyn_hairStrand3_02, swing1LimitDegrees=10, swing2LimitDegrees=10, twistLimitDegrees=2}
 {bone1Name=dyn_hairStrand3_04, bone2Name=dyn_hairStrand3_03, swing1LimitDegrees=10, swing2LimitDegrees=10, twistLimitDegrees=2}
 {
  bone1Name=head
  bone2Name=spine_05
  swing1LimitDegrees=45
  swing1Motion=Free
  swing2LimitDegrees=45
  swing2Motion=Free
  twistLimitDegrees=45
  twistMotion=Free
 }
 {
  bone1Name=clavicle_r
  bone2Name=spine_05
  swing1LimitDegrees=45
  swing1Motion=Free
  swing2LimitDegrees=45
  swing2Motion=Free
  twistLimitDegrees=45
  twistMotion=Free
 }
 {
  bone1Name=clavicle_l
  bone2Name=spine_05
  swing1LimitDegrees=45
  swing1Motion=Free
  swing2LimitDegrees=45
  swing2Motion=Free
  twistLimitDegrees=45
  twistMotion=Free
 }
 {
  bone1Name=upperarm_r
  bone2Name=clavicle_r
  swing1LimitDegrees=45
  swing1Motion=Free
  swing2LimitDegrees=45
  swing2Motion=Free
  twistLimitDegrees=45
  twistMotion=Free
 }
]
