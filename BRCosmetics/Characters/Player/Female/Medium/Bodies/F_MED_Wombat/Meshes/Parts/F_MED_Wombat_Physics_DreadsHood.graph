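# Physics asset for the dreads-hood part of the female medium Wombat player body, providing bodies for four simulated hair strands.
# Each strand is a four-bone chain of boxes and capsules hanging from the head, linked by constraints with tight swing and twist limits.
# A small head box anchors the strands and connects through free-rotation constraints to three upper-spine capsules that give the torso collision volume.
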PhysicsAsset /BRCosmetics/Characters/Player/Female/Medium/Bodies/F_MED_Wombat/Meshes/Parts/F_MED_Wombat_Physics_DreadsHood
bodies {
 dyn_hairStrand1_01 [{shapeType=Box, extentX=5, extentY=2, extentZ=2, center=<2.5 0.0139 0.0219>, rotation=<rot -0.2578 -1.4566 -0.1639>}]
 dyn_hairStrand1_02 [{shapeType=Box, extentX=5, extentY=2, extentZ=2, center=<2.5 -0.1338 0.0163>}]
 dyn_hairStrand1_03 [{shapeType=Box, extentX=5, extentY=2, extentZ=2, center=<2.5 0.0255 0.0205>}]
 dyn_hairStrand1_04 [{shapeType=Capsule, length=4.3308, radius=0.9303, center=<2 0 0>, rotation=<rot 90 0 0>}]
 dyn_hairStrand2_01 [{shapeType=Box, extentX=5, extentY=2, extentZ=2, center=<2.5 -0.0641 0.0019>, rotation=<rot -0.2578 -1.4566 -0.1639>}]
 dyn_hairStrand2_02 [{shapeType=Box, extentX=5, extentY=2, extentZ=2, center=<2.5 0.0204 -0.0522>}]
 dyn_hairStrand2_03 [{shapeType=Box, extentX=5, extentY=2, extentZ=2, center=<2.5 -0.202 0.136>}]
 dyn_hairStrand2_04 [{shapeType=Capsule, length=4.4274, radius=1.0587, center=<2.3 0 0>, rotation=<rot 90 0 0>}]
 dyn_hairStrand3_01 [{shapeType=Box, extentX=5, extentY=2, extentZ=2, center=<2.5 -0.0139 0.013>, rotation=<rot -0.2578 -1.4566 -0.1639>}]
 dyn_hairStrand3_02 [{shapeType=Capsule, length=2, radius=1.5, shapeName=dyn_hairStrand3_02_capsule, center=<2.5 0 0>, rotation=<rot 90 0 0>}]
 dyn_hairStrand3_03 [{shapeType=Capsule, length=2, radius=1.5, shapeName=dyn_hairStrand3_03_capsule, center=<2.5 0 0>, rotation=<rot 90 0 0>}]
 dyn_hairStrand3_04 [{shapeType=Capsule, length=2, radius=1.5, shapeName=dyn_hairStrand3_04_capsule, center=<2.5 0 0>, rotation=<rot 90 0 0>}]
 dyn_hairStrand4_01 [{shapeType=Box, extentX=5, extentY=2, extentZ=2, center=<2.5 0.0144 -0.0604>, rotation=<rot -0.2578 -1.4566 -0.1639>}]
 dyn_hairStrand4_02 [{shapeType=Capsule, length=2, radius=1.5, shapeName=dyn_hairStrand4_02_capsule, center=<2.5 0 0>, rotation=<rot 90 0 0>}]
 dyn_hairStrand4_03 [{shapeType=Capsule, length=2, radius=1.5, shapeName=dyn_hairStrand4_03_capsule, center=<2.5 0 0>, rotation=<rot 90 0 0>}]
 dyn_hairStrand4_04 [{shapeType=Capsule, length=2, radius=1.5, shapeName=dyn_hairStrand4_04_capsule, center=<2.5 0 0>, rotation=<rot 90 0 0>}]
 head [{shapeType=Box, extentX=1, extentY=1, extentZ=1}]
 spine_03 [{shapeType=Capsule, length=10.6864, radius=4.8614, center=<12.334 10.4176 0.0001>, rotation=<rot 0 -90 0>}]
 spine_04 [{shapeType=Capsule, length=18.7384, radius=5.5409, center=<6.996 6.6802 -0.0001>, rotation=<rot 0 -90 0>}]
 spine_05 [{shapeType=Capsule, length=7.9339, radius=12.1112, center=<-1.818 -0.8002 0>}]
}
constraints [
 {bone1Name=dyn_hairStrand1_01, bone2Name=head, swing1LimitDegrees=5, swing2LimitDegrees=5, twistLimitDegrees=2}
 {bone1Name=dyn_hairStrand1_02, bone2Name=dyn_hairStrand1_01, swing1LimitDegrees=5, swing2LimitDegrees=5, twistLimitDegrees=2}
 {bone1Name=dyn_hairStrand1_03, bone2Name=dyn_hairStrand1_02, swing1LimitDegrees=5, swing2LimitDegrees=5, twistLimitDegrees=2}
 {bone1Name=dyn_hairStrand1_04, bone2Name=dyn_hairStrand1_03, swing1LimitDegrees=5, swing2LimitDegrees=5, twistLimitDegrees=2}
 {bone1Name=dyn_hairStrand2_01, bone2Name=head, swing1LimitDegrees=5, swing2LimitDegrees=5, twistLimitDegrees=2}
 {bone1Name=dyn_hairStrand2_02, bone2Name=dyn_hairStrand2_01, swing1LimitDegrees=5, swing2LimitDegrees=5, twistLimitDegrees=2}
 {bone1Name=dyn_hairStrand2_03, bone2Name=dyn_hairStrand2_02, swing1LimitDegrees=5, swing2LimitDegrees=5, twistLimitDegrees=2}
 {bone1Name=dyn_hairStrand2_04, bone2Name=dyn_hairStrand2_03, swing1LimitDegrees=5, swing2LimitDegrees=5, twistLimitDegrees=2}
 {bone1Name=dyn_hairStrand4_01, bone2Name=head, swing1LimitDegrees=5, swing2LimitDegrees=5, twistLimitDegrees=2}
 {bone1Name=dyn_hairStrand4_02, bone2Name=dyn_hairStrand4_01, swing1LimitDegrees=5, swing2LimitDegrees=5, twistLimitDegrees=2}
 {bone1Name=dyn_hairStrand4_03, bone2Name=dyn_hairStrand4_02, swing1LimitDegrees=5, swing2LimitDegrees=5, twistLimitDegrees=2}
 {bone1Name=dyn_hairStrand4_04, bone2Name=dyn_hairStrand4_03, swing1LimitDegrees=5, swing2LimitDegrees=5, twistLimitDegrees=2}
 {bone1Name=dyn_hairStrand3_01, bone2Name=head, swing1LimitDegrees=5, swing2LimitDegrees=5, twistLimitDegrees=2}
 {bone1Name=dyn_hairStrand3_02, bone2Name=dyn_hairStrand3_01, swing1LimitDegrees=5, swing2LimitDegrees=5, twistLimitDegrees=2}
 {bone1Name=dyn_hairStrand3_03, bone2Name=dyn_hairStrand3_02, swing1LimitDegrees=5, swing2LimitDegrees=5, twistLimitDegrees=2}
 {bone1Name=dyn_hairStrand3_04, bone2Name=dyn_hairStrand3_03, swing1LimitDegrees=5, swing2LimitDegrees=5, twistLimitDegrees=2}
 {
  bone1Name=head
  bone2Name=spine_05
  swing1LimitDegrees=45
  swing1Motion=Free
  swing2LimitDegrees=45
  swing2Motion=Free
  twistLimitDegrees=45
  twistMotion=Free
 }
 {
  bone1Name=spine_05
  bone2Name=spine_04
  swing1LimitDegrees=45
  swing1Motion=Free
  swing2LimitDegrees=45
  swing2Motion=Free
  twistLimitDegrees=45
  twistMotion=Free
 }
 {
  bone1Name=spine_04
  bone2Name=spine_03
  swing1LimitDegrees=45
  swing1Motion=Free
  swing2LimitDegrees=45
  swing2Motion=Free
  twistLimitDegrees=45
  twistMotion=Free
 }
]
