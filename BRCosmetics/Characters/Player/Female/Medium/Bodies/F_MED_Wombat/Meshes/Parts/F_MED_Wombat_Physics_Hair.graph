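# Physics asset for the hair of the F_MED_Wombat female character body, defining collision and joint limits for its dynamic hair bones.
# Three back-hair chains (center, left, right) of mostly thin boxes hang from small master capsules, with extra capsules at the tips.
# Two side chains attach to the head, using capsules on the left and boxes on the right.
# Hair joints have tight swing/twist limits, while head, neck, upper-torso and arm capsules provide body shapes with free-rotating joints.
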PhysicsAsset /BRCosmetics/Characters/Player/Female/Medium/Bodies/F_MED_Wombat/Meshes/Parts/F_MED_Wombat_Physics_Hair
bodies {
 clavicle_l [{shapeType=Capsule, length=2.9385, radius=6.5711, shapeName=clavicle_l_capsule, center=<5.599 -1.5316 -1.5849>, rotation=<rot -90 0 0>}]
 clavicle_r [{shapeType=Capsule, length=2.9385, radius=6.5711, shapeName=clavicle_l_capsule, center=<-5.599 1.5316 1.5849>, rotation=<rot -90 0 0>}]
 dyn_hairBack_01 [{shapeType=Box, extentX=7, extentY=10, extentZ=1, center=<3.5 0 0>}]
 dyn_hairBack_01_l [{shapeType=Box, extentX=7, extentY=10, extentZ=1, center=<3.5 0 0>}]
 dyn_hairBack_01_r [{shapeType=Box, extentX=7, extentY=10, extentZ=1, center=<3.5 0 0>}]
 dyn_hairBack_02 [{shapeType=Box, extentX=7, extentY=8, extentZ=1, center=<3.5 -1 0>}]
 dyn_hairBack_02_l [{shapeType=Box, extentX=7, extentY=8, extentZ=1, center=<3.5 1 0>}]
 dyn_hairBack_02_r [{shapeType=Box, extentX=7, extentY=10, extentZ=1, center=<3.5 0 0>}]
 dyn_hairBack_03 [
  {
   shapeType=Capsule
   length=3.021
   radius=2.1886
   shapeName=dyn_hairBack_03_r_capsule
   center=<8.9101 -0.7512 0.0978>
   rotation=<rot -82.5631 -161.9883 70.8361>
  }
  {shapeType=Box, extentX=7, extentY=7, extentZ=1, center=<4.2344 -0.5 -1.1072>, rotation=<rot 11.57 0 0>}
 ]
 dyn_hairBack_03_l [
  {shapeType=Capsule, length=2, radius=3.5, shapeName=dyn_hairBack_03_l_capsule, center=<6.0835 1.184 2.0392>, rotation=<rot 90 0 -12.8317>}
  {shapeType=Box, extentX=7, extentY=7.1227, extentZ=1, center=<3.3271 0.5 0.3393>}
 ]
 dyn_hairBack_03_r [
  {
   shapeType=Capsule
   length=2.894
   radius=1.7296
   shapeName=dyn_hairBack_03_r_capsule
   center=<7.4191 2.5665 -0.8184>
   rotation=<rot -0.0001 0 -65.1341>
  }
  {shapeType=Box, extentX=7, extentY=6.6727, extentZ=1, center=<3.4543 0.0076 -0.4923>, rotation=<rot -3.2604 -1.3858 14.5401>}
 ]
 dyn_hairBack_mstr [{shapeType=Capsule, length=1, radius=1, shapeName=dyn_hairBack_mstr_capsule}]
 dyn_hairBack_mstr_l [{shapeType=Capsule, length=1, radius=1, shapeName=dyn_hairBack_mstr_l_capsule}]
 dyn_hairBack_mstr_r [{shapeType=Capsule, length=1, radius=1, shapeName=dyn_hairBack_mstr_r_capsule}]
 dyn_hair_01_l [{
   shapeType=Capsule
   length=4.75
   radius=2
   shapeName=dyn_hair_01_l_capsule
   center=<2.012 -0.0001 -0.0203>
   rotation=<rot -89.4215 179.9988 -179.9988>
  }]
 dyn_hair_01_r [{shapeType=Box, extentX=7, extentY=5, extentZ=1, center=<3.5 0 0>}]
 dyn_hair_02_l [{
   shapeType=Capsule
   length=0.25
   radius=2.75
   shapeName=dyn_hair_02_l_capsule
   center=<2.7181 -0.0001 -0.0359>
   rotation=<rot -89.206 179.9994 -179.9995>
  }]
 dyn_hair_02_r [{shapeType=Box, extentX=7, extentY=5, extentZ=1, center=<3.5 0 0>}]
 dyn_hair_03_l [{
   shapeType=Capsule
   length=4.25
   radius=2
   shapeName=dyn_hair_03_l_capsule
   center=<3.799 0 -0.2685>
   rotation=<rot -85.8957 179.9998 -179.9999>
  }]
 dyn_hair_03_r [{shapeType=Box, extentX=7, extentY=5, extentZ=1, center=<3.5 0 0>}]
 head [
  {shapeType=Capsule, length=10.4724, radius=6.3282, center=<-2.92 -0.1938 0.2874>, rotation=<rot 90 0 -10>}
  {shapeType=Capsule, length=5, radius=7, shapeName=head_capsule, center=<-6.0262 3.8133 0>, rotation=<rot 0 0 90>}
 ]
 neck_01 [{shapeType=Capsule, length=3.5, radius=5.289, center=<-2.5189 -0.114 0>}]
 spine_04 [{shapeType=Capsule, length=20.0635, radius=2.9365, shapeName=spine_04_capsule, center=<20.042 2.4895 -0.7347>}]
 spine_05 [
  {shapeType=Capsule, length=7.0013, radius=10.1262, center=<0.8956 -0.5834 0.0001>}
  {shapeType=Capsule, length=13.1748, radius=5.2559, shapeName=spine_05_capsule, center=<5.8521 -2.749 0>}
 ]
 upperarm_l [{shapeType=Capsule, length=18.1106, radius=5.5557, shapeName=upperarm_l_capsule, center=<9.0269 0.5727 -0.9985>, rotation=<rot -90 0 0.0001>}]
 upperarm_r [{
   shapeType=Capsule
   length=21.2862
   radius=5.5134
   shapeName=upperarm_l_capsule
   center=<-9.0269 -0.5728 0.9985>
   rotation=<rot -90 -179.9999 0.0001>
  }]
}
constraints [
 {bone1Name=dyn_hairBack_02, bone2Name=dyn_hairBack_01, swing1LimitDegrees=10, swing2LimitDegrees=20, twistLimitDegrees=5}
 {bone1Name=dyn_hairBack_02_l, bone2Name=dyn_hairBack_01_l, swing1LimitDegrees=10, swing2LimitDegrees=20, twistLimitDegrees=5}
 {bone1Name=dyn_hair_02_r, bone2Name=dyn_hair_01_r, swing1LimitDegrees=25, swing2LimitDegrees=25, twistLimitDegrees=5}
 {bone1Name=dyn_hair_03_r, bone2Name=dyn_hair_02_r, swing1LimitDegrees=25, swing2LimitDegrees=25, twistLimitDegrees=5}
 {bone1Name=dyn_hair_02_l, bone2Name=dyn_hair_01_l, swing1LimitDegrees=20, swing2LimitDegrees=20, twistLimitDegrees=0}
 {bone1Name=dyn_hair_03_l, bone2Name=dyn_hair_02_l, swing1LimitDegrees=20, swing2LimitDegrees=20, twistLimitDegrees=0}
 {bone1Name=dyn_hairBack_02_r, bone2Name=dyn_hairBack_01_r, swing1LimitDegrees=10, swing2LimitDegrees=20, twistLimitDegrees=0}
 {bone1Name=dyn_hair_01_r, bone2Name=head, swing1LimitDegrees=25, swing2LimitDegrees=25, twistLimitDegrees=5}
 {bone1Name=dyn_hair_01_l, bone2Name=head, swing1LimitDegrees=20, swing2LimitDegrees=20, twistLimitDegrees=0}
 {
  bone1Name=head
  bone2Name=neck_01
  swing1LimitDegrees=45
  swing1Motion=Free
  swing2LimitDegrees=45
  swing2Motion=Free
  twistLimitDegrees=45
  twistMotion=Free
 }
 {
  bone1Name=neck_01
  bone2Name=spine_05
  swing1LimitDegrees=45
  swing1Motion=Free
  swing2LimitDegrees=45
  swing2Motion=Free
  twistLimitDegrees=45
  twistMotion=Free
 }
 {
  bone1Name=upperarm_r
  bone2Name=clavicle_r
  swing1LimitDegrees=45
  swing1Motion=Free
  swing2LimitDegrees=45
  swing2Motion=Free
  twistLimitDegrees=45
  twistMotion=Free
 }
 {
  bone1Name=clavicle_r
  bone2Name=spine_05
  swing1LimitDegrees=45
  swing1Motion=Free
  swing2LimitDegrees=45
  swing2Motion=Free
  twistLimitDegrees=45
  twistMotion=Free
 }
 {
  bone1Name=clavicle_l
  bone2Name=spine_05
  swing1LimitDegrees=45
  swing1Motion=Free
  swing2LimitDegrees=45
  swing2Motion=Free
  twistLimitDegrees=45
  twistMotion=Free
 }
 {
  bone1Name=upperarm_l
  bone2Name=clavicle_l
  swing1LimitDegrees=45
  swing1Motion=Free
  swing2LimitDegrees=45
  swing2Motion=Free
  twistLimitDegrees=45
  twistMotion=Free
 }
 {bone1Name=dyn_hairBack_03_r, bone2Name=dyn_hairBack_02_r, swing1LimitDegrees=15, swing2LimitDegrees=20, twistLimitDegrees=0}
 {bone1Name=dyn_hairBack_03, bone2Name=dyn_hairBack_02, swing1LimitDegrees=10, swing2LimitDegrees=20, twistLimitDegrees=5}
 {bone1Name=dyn_hairBack_03_l, bone2Name=dyn_hairBack_02_l, swing1LimitDegrees=15, swing2LimitDegrees=20, twistLimitDegrees=5}
 {bone1Name=dyn_hairBack_01_r, bone2Name=dyn_hairBack_mstr_r, swing1LimitDegrees=20, swing2LimitDegrees=20, twistLimitDegrees=0}
 {bone1Name=dyn_hairBack_01, bone2Name=dyn_hairBack_mstr, swing1LimitDegrees=10, swing2LimitDegrees=25, twistLimitDegrees=5}
 {bone1Name=dyn_hairBack_01_l, bone2Name=dyn_hairBack_mstr_l, swing1LimitDegrees=15, swing2LimitDegrees=25, twistLimitDegrees=5}
]
