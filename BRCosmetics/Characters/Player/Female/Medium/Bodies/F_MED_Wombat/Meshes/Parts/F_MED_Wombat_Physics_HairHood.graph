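# Physics asset for the hair/hood part of the female medium Wombat player cosmetic, defining bodies and joints for its dynamic hair bones.
# Five three-bone chains hang from the head: three flat-box back chains and two narrower side chains with capsule tips.
# Capsules on head, neck and upper spine form the body volumes, linked by free-rotation joints.
# Hair joints limit swing and twist, tighter on the sides, except the right tip, which matches the looser back joints.
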PhysicsAsset /BRCosmetics/Characters/Player/Female/Medium/Bodies/F_MED_Wombat/Meshes/Parts/F_MED_Wombat_Physics_HairHood
bodies {
 dyn_hairBack_01 [{shapeType=Box, extentX=7, extentY=10, extentZ=1, center=<3.5 0 0>}]
 dyn_hairBack_01_l [{shapeType=Box, extentX=7, extentY=10, extentZ=1, center=<3.5 0 0>}]
 dyn_hairBack_01_r [{shapeType=Box, extentX=7, extentY=10, extentZ=1, center=<3.5 0 0>}]
 dyn_hairBack_02 [{shapeType=Box, extentX=7, extentY=10, extentZ=1, center=<3.5 0 0>}]
 dyn_hairBack_02_l [{shapeType=Box, extentX=7, extentY=10, extentZ=1, center=<3.5 0 0>}]
 dyn_hairBack_02_r [{shapeType=Box, extentX=7, extentY=10, extentZ=1, center=<3.5 0 0>}]
 dyn_hairBack_03 [{shapeType=Box, extentX=7, extentY=10, extentZ=1, center=<3.5 0 0>}]
 dyn_hairBack_03_l [{shapeType=Box, extentX=7, extentY=10, extentZ=1, center=<3.5 0 0>}]
 dyn_hairBack_03_r [{shapeType=Box, extentX=7, extentY=10, extentZ=1, center=<3.5 0 0>}]
 dyn_hair_01_l [{shapeType=Box, extentX=7, extentY=5, extentZ=1, center=<3.5 0 0>}]
 dyn_hair_01_r [{shapeType=Box, extentX=7, extentY=5, extentZ=1, center=<3.5 0 0>}]
 dyn_hair_02_l [{shapeType=Box, extentX=7, extentY=5, extentZ=1, center=<3.5 0 0>}]
 dyn_hair_02_r [{shapeType=Box, extentX=7, extentY=5, extentZ=1, center=<3.5 0 0>}]
 dyn_hair_03_l [{shapeType=Capsule, length=5.1304, radius=1.8355, center=<3.0472 0 -0.0001>, rotation=<rot 90 0 0>}]
 dyn_hair_03_r [{shapeType=Capsule, length=5.1304, radius=1.8355, center=<2.5864 0 0.0002>, rotation=<rot 90 0 -179.9999>}]
 head [{shapeType=Capsule, length=10.4724, radius=7.3282, center=<-2.92 -0.1938 0.2874>, rotation=<rot 90 0 -10>}]
 neck_01 [{shapeType=Capsule, length=7.5, radius=5.289, center=<-2.5189 -0.114 0>}]
 spine_05 [
  {shapeType=Capsule, length=15.696, radius=10.1262, center=<0.4504 0 0>}
  {shapeType=Capsule, length=26.5, radius=3.5, center=<10.559 -0.877 0>}
 ]
}
constraints [
 {bone1Name=dyn_hairBack_02, bone2Name=dyn_hairBack_01, swing1LimitDegrees=35, swing2LimitDegrees=40, twistLimitDegrees=5}
 {bone1Name=dyn_hairBack_03, bone2Name=dyn_hairBack_02, swing1LimitDegrees=35, swing2LimitDegrees=40, twistLimitDegrees=5}
 {bone1Name=dyn_hairBack_02_l, bone2Name=dyn_hairBack_01_l, swing1LimitDegrees=35, swing2LimitDegrees=40, twistLimitDegrees=5}
 {bone1Name=dyn_hairBack_03_l, bone2Name=dyn_hairBack_02_l, swing1LimitDegrees=35, swing2LimitDegrees=40, twistLimitDegrees=5}
 {bone1Name=dyn_hair_02_r, bone2Name=dyn_hair_01_r, swing1LimitDegrees=15, swing2LimitDegrees=5, twistLimitDegrees=5}
 {bone1Name=dyn_hair_02_l, bone2Name=dyn_hair_01_l, swing1LimitDegrees=15, swing2LimitDegrees=5, twistLimitDegrees=5}
 {bone1Name=dyn_hair_03_l, bone2Name=dyn_hair_02_l, swing1LimitDegrees=15, swing2LimitDegrees=5, twistLimitDegrees=5}
 {bone1Name=dyn_hairBack_02_r, bone2Name=dyn_hairBack_01_r, swing1LimitDegrees=35, swing2LimitDegrees=40, twistLimitDegrees=5}
 {bone1Name=dyn_hairBack_03_r, bone2Name=dyn_hairBack_02_r, swing1LimitDegrees=35, swing2LimitDegrees=40, twistLimitDegrees=5}
 {bone1Name=dyn_hairBack_01, bone2Name=head, swing1LimitDegrees=25, swing2LimitDegrees=25, twistLimitDegrees=5}
 {bone1Name=dyn_hairBack_01_l, bone2Name=head, swing1LimitDegrees=25, swing2LimitDegrees=25, twistLimitDegrees=5}
 {bone1Name=dyn_hair_01_r, bone2Name=head, swing1LimitDegrees=25, swing2LimitDegrees=25, twistLimitDegrees=5}
 {bone1Name=dyn_hair_01_l, bone2Name=head, swing1LimitDegrees=25, swing2LimitDegrees=25, twistLimitDegrees=5}
 {bone1Name=dyn_hairBack_01_r, bone2Name=head, swing1LimitDegrees=25, swing2LimitDegrees=25, twistLimitDegrees=5}
 {
  bone1Name=head
  bone2Name=neck_01
  swing1LimitDegrees=45
  swing1Motion=Free
  swing2LimitDegrees=45
  swing2Motion=Free
  twistLimitDegrees=45
  twistMotion=Free
 }
 {
  bone1Name=neck_01
  bone2Name=spine_05
  swing1LimitDegrees=45
  swing1Motion=Free
  swing2LimitDegrees=45
  swing2Motion=Free
  twistLimitDegrees=45
  twistMotion=Free
 }
 {bone1Name=dyn_hair_03_r, bone2Name=dyn_hair_02_r, swing1LimitDegrees=35, swing2LimitDegrees=40, twistLimitDegrees=5}
]
