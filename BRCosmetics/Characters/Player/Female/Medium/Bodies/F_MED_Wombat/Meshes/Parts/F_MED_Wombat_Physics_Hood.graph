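# Physics asset for the hood part of the female medium Wombat player cosmetic body, defining its ragdoll-style simulation setup.
# It holds box bodies on the head bone and two hood bones, dyn_hoodTop and dyn_hood_01.
# Each hood body is constrained directly to the head, rather than chained, with limited swing and tight twist so the hood can sway within a narrow cone.
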PhysicsAsset /BRCosmetics/Characters/Player/Female/Medium/Bodies/F_MED_Wombat/Meshes/Parts/F_MED_Wombat_Physics_Hood
bodies {
 dyn_hoodTop [{shapeType=Box, extentX=5, extentY=5, extentZ=5, center=<2.5 0 0>}]
 dyn_hood_01 [{shapeType=Box, extentX=5, extentY=5, extentZ=1, center=<2.5 0 0>}]
 head [{shapeType=Box, extentX=1, extentY=1, extentZ=1}]
}
constraints [
 {bone1Name=dyn_hood_01, bone2Name=head, swing1LimitDegrees=25, swing2LimitDegrees=15, twistLimitDegrees=5}
 {bone1Name=dyn_hoodTop, bone2Name=head, swing1LimitDegrees=25, swing2LimitDegrees=25, twistLimitDegrees=5}
]
