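# Physics asset for the necklace part of the F_MED_Wombat female medium player body in BRCosmetics.
# It has two bodies: a small capsule on the dyn_necklace_01 bone and a wide, thin box on the dyn_necklace_mstr bone.
# A single constraint with limited swing and twist joins the two bones, restricting how far the necklace bone can rotate relative to the master.
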PhysicsAsset /BRCosmetics/Characters/Player/Female/Medium/Bodies/F_MED_Wombat/Meshes/Parts/F_MED_Wombat_Physics_Necklace
bodies {
 dyn_necklace_01 [{shapeType=Capsule, length=1, radius=1, shapeName=dyn_necklace_01_capsule, center=<1.5 0 0.5>, rotation=<rot 90 0 0>}]
 dyn_necklace_mstr [{shapeType=Box, extentX=9.01, extentY=11.01, extentZ=1.01, center=<1.0356 0 -0.8057>}]
}
constraints [{bone1Name=dyn_necklace_01, bone2Name=dyn_necklace_mstr, swing1LimitDegrees=45, swing2LimitDegrees=25, twistLimitDegrees=15}]
