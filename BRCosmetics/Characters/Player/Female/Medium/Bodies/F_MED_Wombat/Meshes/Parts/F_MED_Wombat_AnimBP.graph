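# Post-process-style Anim Blueprint for the Wombat female body cosmetic, simulating hair, dreads, hood and necklace over the parent mesh's copied pose.
# A cached necklace constraint plus hood/necklace RigidBody pose feeds hair chains blended by front-end and poncho state, with hood physics assets and bone offsets.
# The EventGraph sets poncho from the Stage2 variant tag, hides face hair bones, resets dynamics, and drives gravity with Perlin-noise wind while swim-sprinting or skydiving.
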AnimBlueprint /BRCosmetics/Characters/Player/Female/Medium/Bodies/F_MED_Wombat/Meshes/Parts/F_MED_Wombat_AnimBP
target_skeleton=&/F_MED_Wombat_Skeleton
graphs [
 AnimGraph {nodes [
   n1 = Get neck_rotating_up() -> (neck_rotating_up:double)
   n2 = "Transform (Modify) Bone - Bone: dyn_hairBack_mstr"(Alpha=@n1, Translation="0,0,0", Rotation="0,-75,15", Scale="1,1,1") {
    rotationMode=BMM_Additive
    translationMode=BMM_Additive
    alphaBoolBlend {blendOption=Linear}
    boneToModify {boneName=dyn_hairBack_mstr}
    rotation {roll=15, yaw=-75}
   }
   n3 = "Transform (Modify) Bone - Bone: dyn_hairBack_mstr_r"(ComponentPose=@n2, Alpha=@n1, Translation="0,0,0", Rotation="0,-75,10", Scale="1,1,1") {
    rotationMode=BMM_Additive
    translationMode=BMM_Additive
    alphaBoolBlend {blendOption=Linear}
    boneToModify {boneName=dyn_hairBack_mstr_r}
    rotation {roll=10, yaw=-75}
   }
   "Transform (Modify) Bone - Bone: dyn_hairBack_mstr_l"(ComponentPose=@n3, Alpha=@n1, Translation="0,0,0", Rotation="0,0,20", Scale="1,1,1") {
    rotationMode=BMM_Additive
    translationMode=BMM_Additive
    alphaBoolBlend {blendOption=Linear}
    boneToModify {boneName=dyn_hairBack_mstr_l}
    rotation {roll=20}
   }
   n4 = Get GravityOverride()
   n5 = RigidBody!(
    ComponentPose=Local To Component(LocalPose=Use cached pose 'constConstraint'())
    bAlphaBoolEnabled:bool
    OverrideWorldGravity:Vector=@n4
    ExternalForce
   ) -> (Pose) {
    alphaInputType=Bool
    bClampLinearTranslationLimitToRefPose=true
    bOverrideWorldGravity=true
    overridePhysicsAsset=&/F_MED_Wombat_Physics_Hair
    simulationSpace=BaseBoneSpace
    alphaBoolBlend {blendOption=Linear}
    baseBoneRef {boneName=spine_05}
    componentAppliedLinearAccClamp=<4000 4000 4000>
    componentLinearAccScale=<0.3 0.3 0.3>
    componentLinearVelScale=<0.35 0.35 0.35>
    overrideWorldGravity {z=-980}
   }
   n6 = RigidBody(ComponentPose=@n5, OverrideWorldGravity=@n4) {
    alphaInputType=Bool
    bClampLinearTranslationLimitToRefPose=true
    bOverrideWorldGravity=true
    overridePhysicsAsset=&/F_MED_Wombat_Physics_Dreads
    simulationSpace=BaseBoneSpace
    alphaBoolBlend {blendOption=Linear}
    baseBoneRef {boneName=head}
    componentAppliedLinearAccClamp=<3000 3000 3000>
    componentLinearAccScale=<0.3 0.3 0.3>
    componentLinearVelScale=<0.3 0.3 0.3>
    overrideWorldGravity {z=-980}
   }
   n7 = Component To Local(ComponentPose=@n6)
   n8 = RigidBody(ComponentPose=Local To Component(LocalPose=Use cached pose 'constConstraint'()), OverrideWorldGravity=@n4) {
    alphaInputType=Bool
    bClampLinearTranslationLimitToRefPose=true
    bOverrideWorldGravity=true
    overridePhysicsAsset=&/F_MED_Wombat_Physics_HairHood
    simulationSpace=BaseBoneSpace
    alphaBoolBlend {blendOption=Linear}
    baseBoneRef {boneName=spine_05}
    componentAppliedLinearAccClamp=<3000 3000 3000>
    componentLinearAccScale=<0.45 0.45 0.45>
    componentLinearVelScale=<0.3 0.3 0.3>
    overrideWorldGravity {z=-980}
   }
   n9 = RigidBody(ComponentPose=@n8, OverrideWorldGravity=@n4) {
    alphaInputType=Bool
    bClampLinearTranslationLimitToRefPose=true
    bOverrideWorldGravity=true
    overridePhysicsAsset=&/F_MED_Wombat_Physics_DreadsHood
    simulationSpace=BaseBoneSpace
    alphaBoolBlend {blendOption=Linear}
    baseBoneRef {boneName=head}
    componentAppliedLinearAccClamp=<3000 3000 3000>
    componentLinearAccScale=<0.3 0.3 0.3>
    componentLinearVelScale=<0.3 0.3 0.3>
    overrideWorldGravity {z=-980}
   }
   n10 = "Transform (Modify) Bone - Bone: dyn_hairBack_01"!(ComponentPose=@n9, bAlphaBoolEnabled:bool=True, Translation="0,0,0", Rotation="0,0,-20", Scale="1,1,1") -> (Pose) {
    alphaInputType=Bool
    rotationMode=BMM_Additive
    alphaBoolBlend {blendOption=Linear}
    boneToModify {boneName=dyn_hairBack_01}
    rotation {roll=-20}
   }
   n11 = "Transform (Modify) Bone - Bone: dyn_hairBack_01_l"!(ComponentPose=@n10, bAlphaBoolEnabled:bool=True, Translation="0,0,0", Rotation="-10,0,-15", Scale="1,1,1") -> (Pose) {
    alphaInputType=Bool
    rotationMode=BMM_Additive
    alphaBoolBlend {blendOption=Linear}
    boneToModify {boneName=dyn_hairBack_01_l}
    rotation {pitch=-10, roll=-15}
   }
   n12 = "Transform (Modify) Bone - Bone: dyn_hairBack_01_r"!(ComponentPose=@n11, bAlphaBoolEnabled:bool=True, Translation="0,0,0", Rotation="0,0,-15", Scale="1,1,1") -> (Pose) {
    alphaInputType=Bool
    rotationMode=BMM_Additive
    alphaBoolBlend {blendOption=Linear}
    boneToModify {boneName=dyn_hairBack_01_r}
    rotation {roll=-15}
   }
   n13 = Component To Local(ComponentPose=@n12)
   n14 = Get ponchoActive() -> (ponchoActive:bool)
   n15 = "Transform (Modify) Bone - Bone: dyn_hairBack_01"(
    ComponentPose=Local To Component(LocalPose=Use cached pose 'constConstraint'())
    bAlphaBoolEnabled=True
    Translation="0,0,0"
    Rotation="0,0,-20"
    Scale="1,1,1"
   ) {
    alphaInputType=Bool
    rotationMode=BMM_Additive
    alphaBoolBlend {blendOption=Linear}
    boneToModify {boneName=dyn_hairBack_01}
    rotation {roll=-20}
   }
   n16 = "Transform (Modify) Bone - Bone: dyn_hairBack_01_l"(ComponentPose=@n15, bAlphaBoolEnabled=True, Translation="0,0,0", Rotation="-10,0,-15", Scale="1,1,1") {
    alphaInputType=Bool
    rotationMode=BMM_Additive
    alphaBoolBlend {blendOption=Linear}
    boneToModify {boneName=dyn_hairBack_01_l}
    rotation {pitch=-10, roll=-15}
   }
   n17 = "Transform (Modify) Bone - Bone: dyn_hairBack_01_r"(ComponentPose=@n16, bAlphaBoolEnabled=True, Translation="0,0,0", Rotation="0,0,-15", Scale="1,1,1") {
    alphaInputType=Bool
    rotationMode=BMM_Additive
    alphaBoolBlend {blendOption=Linear}
    boneToModify {boneName=dyn_hairBack_01_r}
    rotation {roll=-15}
   }
   n18 = "Transform (Modify) Bone - Bone: dyn_hair_01_l"!(ComponentPose=@n17, bAlphaBoolEnabled:bool=True, Translation="0,0,0", Rotation="-10,0,0", Scale="1,1,1") -> (Pose) {
    alphaInputType=Bool
    rotationMode=BMM_Additive
    alphaBoolBlend {blendOption=Linear}
    boneToModify {boneName=dyn_hair_01_l}
    rotation {pitch=-10}
   }
   n19 = "Transform (Modify) Bone - Bone: dyn_hair_01_r"!(ComponentPose=@n18, bAlphaBoolEnabled:bool=True, Translation="0,0,0", Rotation="18,0,0", Scale="1,1,1") -> (Pose) {
    alphaInputType=Bool
    rotationMode=BMM_Additive
    alphaBoolBlend {blendOption=Linear}
    boneToModify {boneName=dyn_hair_01_r}
    rotation {pitch=18}
   }
   n20 = RigidBody(ComponentPose=@n19, OverrideWorldGravity=@n4) {
    alphaInputType=Bool
    bClampLinearTranslationLimitToRefPose=true
    bOverrideWorldGravity=true
    overridePhysicsAsset=&F_MED_Wombat_Physics_HairHood
    simulationSpace=BaseBoneSpace
    alphaBoolBlend {blendOption=Linear}
    baseBoneRef {boneName=spine_05}
    componentAppliedLinearAccClamp=<3000 3000 3000>
    componentLinearAccScale=<0.45 0.45 0.45>
    componentLinearVelScale=<0.3 0.3 0.3>
    overrideWorldGravity {z=-980}
   }
   n21 = RigidBody(ComponentPose=@n20, OverrideWorldGravity=@n4) {
    alphaInputType=Bool
    bClampLinearTranslationLimitToRefPose=true
    bOverrideWorldGravity=true
    overridePhysicsAsset=&F_MED_Wombat_Physics_DreadsHood
    simulationSpace=BaseBoneSpace
    alphaBoolBlend {blendOption=Linear}
    baseBoneRef {boneName=head}
    componentAppliedLinearAccClamp=<3000 3000 3000>
    componentLinearAccScale=<0.3 0.3 0.3>
    componentLinearVelScale=<0.3 0.3 0.3>
    overrideWorldGravity {z=-980}
   }
   n22 = Component To Local(ComponentPose=@n21)
   n23 = RigidBody(ComponentPose=Local To Component(LocalPose=Use cached pose 'constConstraint'()), OverrideWorldGravity=@n4) {
    alphaInputType=Bool
    bClampLinearTranslationLimitToRefPose=true
    bOverrideWorldGravity=true
    overridePhysicsAsset=&F_MED_Wombat_Physics_Hair
    simulationSpace=BaseBoneSpace
    alphaBoolBlend {blendOption=Linear}
    baseBoneRef {boneName=spine_05}
    componentAppliedLinearAccClamp=<4000 4000 4000>
    componentLinearAccScale=<0.3 0.3 0.3>
    componentLinearVelScale=<0.35 0.35 0.35>
    overrideWorldGravity {z=-980}
   }
   n24 = RigidBody(ComponentPose=@n23, OverrideWorldGravity=@n4) {
    alphaInputType=Bool
    bClampLinearTranslationLimitToRefPose=true
    bOverrideWorldGravity=true
    overridePhysicsAsset=&F_MED_Wombat_Physics_Dreads
    simulationSpace=BaseBoneSpace
    alphaBoolBlend {blendOption=Linear}
    baseBoneRef {boneName=head}
    componentAppliedLinearAccClamp=<3000 3000 3000>
    componentLinearAccScale=<0.3 0.3 0.3>
    componentLinearVelScale=<0.3 0.3 0.3>
    overrideWorldGravity {z=-980}
   }
   n25 = Component To Local(ComponentPose=@n24)
   n26 = Get ponchoActive()
   Output Pose(Result=Blend Poses by bool(
     BlendPose_0=Blend Poses by bool(BlendPose_0=@n13, BlendPose_1=@n7, bActiveValue=@n14) {blendType=Linear}
     BlendPose_1=Blend Poses by bool(BlendPose_0=@n22, BlendPose_1=@n25, bActiveValue=@n26) {blendType=Linear}
     bActiveValue=Get bIsInFrontEnd()
    ) {blendType=Linear}) {layerGroup=None}
   n27 = "Constraint - dyn_necklace_mstr (PPPPPPPP)"!(
    ComponentPose=Local To Component(LocalPose=Copy Pose From Mesh() {bUseAttachedParent=true})
    Alpha=1
    ConstraintWeights_0=0
    ConstraintWeights_1:float=0.0253
    ConstraintWeights_2:float=0.0684
    ConstraintWeights_3:float=0.3557
    ConstraintWeights_4:float=0.5417
    ConstraintWeights_5:float=0.0066
    ConstraintWeights_6:float=0.0001
    ConstraintWeights_7:float=0.0023
   ) -> (Pose) {
    alphaBoolBlend {blendOption=Linear}
    boneToModify {boneName=dyn_necklace_mstr}
    constraintSetup [
     {targetBone {boneName=head}}
     {targetBone {boneName=clavicle_l}}
     {targetBone {boneName=spine_03}}
     {targetBone {boneName=spine_04}}
     {targetBone {boneName=spine_05}}
     {targetBone {boneName=clavicle_r}}
     {targetBone {boneName=neck_02}}
     {targetBone {boneName=neck_01}}
    ]
    constraintWeights [1,1,1,1,1,1,1,1]
   }
   n28 = RigidBody!(ComponentPose=@n27, Alpha:float, OverrideWorldGravity=@n4, ExternalForce) -> (Pose) {
    bClampLinearTranslationLimitToRefPose=true
    bOverrideWorldGravity=true
    overridePhysicsAsset=&/F_MED_Wombat_Physics_Hood
    simulationSpace=BaseBoneSpace
    alphaBoolBlend {blendOption=Linear}
    baseBoneRef {boneName=head}
    componentAppliedLinearAccClamp=<3000 3000 3000>
    componentLinearAccScale=<0.3 0.3 0.3>
    componentLinearVelScale=<0.3 0.3 0.3>
    overrideWorldGravity {z=-980}
   }
   n29 = RigidBody(ComponentPose=@n28, OverrideWorldGravity=@n4) {
    bClampLinearTranslationLimitToRefPose=true
    bOverrideWorldGravity=true
    overridePhysicsAsset=&/F_MED_Wombat_Physics_Necklace
    simulationSpace=BaseBoneSpace
    alphaBoolBlend {blendOption=Linear}
    baseBoneRef {boneName=spine_05}
    componentAppliedLinearAccClamp=<3000 3000 3000>
    componentLinearAccScale=<0.3 0.3 0.3>
    componentLinearVelScale=<0.3 0.3 0.3>
    overrideWorldGravity {z=-980}
   }
   Save cached pose 'constConstraint'(Pose=Component To Local(ComponentPose=@n29)) {cacheName=constConstraint}
   Branch()
  ]}
 EventGraph {nodes [
   Event BlueprintUpdateAnimation()
   > n1 = Sequence!(execute) -> (then_0, then_1, then_2:Exec, then_3:Exec, then_4:Exec)
   > n2 = Branch(Condition=Get bIsSwimSprinting() -> (bIsSwimSprinting:bool))
   n3 = GetSocketRotation!(self:Scene Component Object Reference=Get Skel_Mesh(), InSocketName:Name=pelvis) -> (ReturnValue:Rotator)
   n4 = PerlinNoise1D!(Value:float=GetGameTimeInSeconds()) -> (ReturnValue:float)
   n5 = float * float(A="Sin (Degrees)"!(A:double=float * float(A=@n4, B=800)) -> (ReturnValue:double), B=-400)
   n6 = vector + vector(
    A:Vector=vector * float(A=GetForwardVector!(InRot:Rotator=@n3) -> (ReturnValue:Vector), B=-300)
    B:Vector=MakeVector(Y=0, Z=float + float(A=@n5, B=200) {is=title K2Node_CommutativeAssociativeBinaryOperator})
   ) -> (ReturnValue:Vector) {is=title K2Node_CommutativeAssociativeBinaryOperator}
   > Set GravityOverride(GravityOverride=@n6)
   n7 = Branch(execute=@n2.else, Condition=Get bIsSkyDiving())
   > n8 = Branch(Condition=Get bIsParachuteOpened())
   Set GravityOverride(execute=[@n7.else, @n8.then], GravityOverride="0, 0,-980")
   n9 = Branch(execute=@n8.else, Condition="float > float"(A=Get PawnVelocityZ(), B=50) {is=title K2Node_CallFunction})
   > Set GravityOverride(GravityOverride="0, 0,-100")
   Set GravityOverride(execute=@n9.else, GravityOverride=@n6)
   n10 = Branch(execute=@n1.then_1, Condition=Get ponchoActive())
   > Do Once()
   > Delay(Duration=0.15)
   > ResetDynamics(InTeleportType=TeleportPhysics)
   > ResetDynamics(InTeleportType=ResetPhysics)
   n11 = Get faceAcc() -> (faceAcc:Skeletal Mesh Component Object Reference)
   > HideBoneByName!(execute, self:Skinned Mesh Component Object Reference=@n11, BoneName:Name=dyn_hair_01_r, PhysBodyOption:EPhysBodyOp Enum=PBO_Term) -> (then)
   > HideBoneByName(self=@n11, BoneName=dyn_hairStrand4_01, PhysBodyOption=PBO_Term)
   > HideBoneByName(self=@n11, BoneName=dyn_hair_01_l, PhysBodyOption=PBO_Term)
   Do Once(execute=@n10.else)
   > Delay(Duration=0.15)
   > ResetDynamics(InTeleportType=TeleportPhysics)
   > ResetDynamics(InTeleportType=ResetPhysics)
   n12 = Branch(execute=@n1.then_2, Condition=Get bIsSkyDiving())
   > Set neck_rotating_up(neck_rotating_up:double=0) -> (Output_Get:double)
   Set neck_rotating_up(
    execute=@n12.else
    neck_rotating_up=GetAlphaFromReferenceTransformDelta(
     SkeletalMeshComponent=Get Skel_Mesh()
     BoneName=head
     BaseName=neck_01
     DeltaComponentType=QuaternionTwist
     InputRange=(X=0,Y=25)
     TwistAxis="0,0,1"
    )
   )
   n13 = Branch(execute=@n1.then_3, Condition=OR Boolean!(A=Get bIsCrouchSprinting(), B=Get bIsSprinting(), C:bool=Get bIsJumping()) -> (ReturnValue))
   > Set Sprinting(Sprinting:double=1) -> (Output_Get:double)
   Set Sprinting(execute=@n13.else, Sprinting=0)
   Event BlueprintInitializeAnimation()
   > Set Skel_Mesh(Skel_Mesh=GetOwnerSkeletalMeshComponent())
   Event OnCharacterCustomizationLoaded()
   > n14 = Sequence!(execute) -> (then_0, then_1, then_2)
   > n15 = Cast To FortPlayerPawn(Object=GetOwner(self=GetPartSkeletalMeshComponent()))
   > Set faceAcc(faceAcc:Skeletal Mesh Component Object Reference=GetSkeletalMeshForPartType(self=@n15, PartType=Face)) -> (Output_Get:Skeletal Mesh Component Object Reference)
   n16 = Branch(execute=@n14.then_1, Condition=CharacterItemDefinitionHasAnyMetaTag(Tags=Make Array("[0]"=(TagName="Cosmetics.Variant.Property.Stage2"))))
   > Set ponchoActive(ponchoActive:bool=true) -> (Output_Get:bool)
   Set ponchoActive(execute=@n16.else, ponchoActive=false)
  ]}
]
variables [
 faceAcc {type=SkeletalMeshComponent}
 FrontEnd {type=boolean}
 GravityOverride {type=Vector}
 neck_rotating_up {type=number}
 ponchoActive {type=boolean}
 Skel_Mesh {type=SkeletalMeshComponent}
 Sprinting {type=number}
]
